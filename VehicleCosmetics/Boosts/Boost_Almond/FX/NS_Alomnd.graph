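# NS_Alomnd is the Niagara effect for the Almond vehicle boost cosmetic, emitting from skeletal-mesh sockets.
# One emitter multiplexes spawn groups through per-group switches and visibility tags, producing tumbling, colliding mesh debris plus sprite layers.
# A second emitter adds short-lived hot-colored sprites stirred by curl noise, while an older five-group variant is disabled.
# Frontend-preview switches and VC_Mask inputs retune rates, speeds and looks per context.
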NiagaraSystem /VehicleCosmetics/Boosts/Boost_Almond/FX/NS_Alomnd
dependencies {
 usedDataInterfaces [
  &NiagaraDataInterfaceCurve
  &NiagaraDataInterfaceVectorCurve
  /Script/Niagara.NiagaraDataInterfaceArrayBool
  &NiagaraDataInterfaceSkeletalMesh
  &NiagaraDataInterfaceArrayFloat3
 ]
 usedDynamicInputs [&MakeFloatFromVector2d, &SelectVectorFromArray]
 usedModules [
  &self:SystemSpawnScript.SystemScriptSource.SystemScriptGraph.NiagaraNodeAssignment_0.SetVariables_#1
  &self:SystemSpawnScript.SystemScriptSource.SystemScriptGraph.NiagaraNodeAssignment_1.SetVariables_#2
  &self:SystemSpawnScript.SystemScriptSource.SystemScriptGraph.NiagaraNodeAssignment_2.SetVariables_#3
  &SystemState
 ]
}
emitters {
 NE_VC_Boost_Sprite01 {
  bEnabled=false
  emitterUpdateScript [
   &EmitterState()
   &VC_FilteredSocketCount()
   &SpawnRate(
    SpawnRate=&Multiply_Float(
     A=&MakeCustomFloatFromBool(Bool=$User.bIsFrontendPreview, True Float=60, False Float=40)
     B=/VehicleCosmetics/FX/Niagara/Modules/VC_Mask_Floats(Coro Value:NiagaraFloat=1, FrontEnd Value:NiagaraFloat=0.3, FN Value:NiagaraFloat=0.5)
    )
    Use Spawn Probability=false
   )
   &SpawnRate(
    SpawnGroup=1
    SpawnRate=&Multiply_Float(
     A=&MakeCustomFloatFromBool(Bool=$User.bIsFrontendPreview, True Float=100, False Float=200)
     B=&VC_Mask_Floats(Coro Value:NiagaraFloat=1, FrontEnd Value:NiagaraFloat=0.5, FN Value:NiagaraFloat=0.5)
    )
    Use Spawn Probability=false
   ) {module=SpawnRate001}
   &SpawnRate(
    SpawnGroup=2
    SpawnRate=&Multiply_Float(
     A=400
     B=&VC_Mask_Floats(Coro Value:NiagaraFloat=1, FrontEnd Value:NiagaraFloat=0.3, FN Value:NiagaraFloat=0.5)
    )
    Use Spawn Probability=false
   ) {module=SpawnRate002}
   &SpawnRate(
    SpawnGroup=3
    SpawnRate=&Multiply_Float(
     A=150
     B=&VC_Mask_Floats(Coro Value:NiagaraFloat=1, FrontEnd Value:NiagaraFloat=0.5, FN Value:NiagaraFloat=0.5)
    )
    Use Spawn Probability=false
   ) {module=SpawnRate003}
   &SpawnRate(
    SpawnGroup=4
    SpawnRate=&Multiply_Float(
     A=150
     B=&VC_Mask_Floats(Coro Value:NiagaraFloat=1, FrontEnd Value:NiagaraFloat=1, FN Value:NiagaraFloat=0.5)
    )
    Use Spawn Probability=false
   ) {module=SpawnRate004}
   /SetVariables(
    Emitter.DustColor:LinearColor=&MakeCustomLinearColorFromBool(Bool=$User.bIsFrontendPreview, True Color=<rgba 100 33.3333 0 1>, False Color=<rgba 500 104.2281 2.1739 1>)
    Emitter.Sand Alpha Power:NiagaraFloat=&MakeCustomFloatFromBool(Bool=$User.bIsFrontendPreview, True Float=4, False Float=7)
   ) {isSetParameters=true, module=SetVariables_#4}
  ]
  particleSpawnScript [
   &VC_PaintedVariantsOverride() {enabled=false}
   &InitializeParticle(
    Color=&MergeHelper_ColorSwitch(
     0=<rgba 1 1 1 0>
     1=&MakeCustomLinearColorFromBool(Bool=$User.bIsFrontendPreview, True Color=<rgba 100 83.3333 0 1>, False Color=<rgba 5 1.6667 0 10>)
     2=&MakeCustomLinearColorFromBool(Bool=$User.bIsFrontendPreview, True Color=<rgba 1.5 0.7039 0 1>, False Color=<rgba 30 12.125 10.5 1>)
     3=&MakeCustomLinearColorFromBool(Bool=$User.bIsFrontendPreview, True Color=<rgba 10 4.5 4 1>, False Color=<rgba 8 3.4067 2.8 0.1>)
     4=&MakeCustomLinearColorFromBool(Bool=$User.bIsFrontendPreview, True Color=<rgba 30 15.9977 15 1>, False Color=<rgba 500 88.5984 31.145 1>)
     Index=&SpawnGroup()
     Num=::NewEnumerator3
    )
    Lifetime=&MergeHelper_FloatSwitch(
     0=&MakeCustomFloatFromBool(
      Bool=$User.bIsFrontendPreview
      True Float=&RandomRangeFloat(Minimum=0.3, Maximum=1)
      False Float=&RandomRangeFloat(Minimum=0.5, Maximum=1.5)
     )
     1=&MakeCustomFloatFromBool(
      Bool=$User.bIsFrontendPreview
      True Float=&RandomRangeFloat(Minimum=0.1, Maximum=0.3)
      False Float=&RandomRangeFloat(Minimum=0.1, Maximum=0.3)
     )
     2=&RandomRangeFloat(Minimum=0.15, Maximum=0.7)
     3=&MakeCustomFloatFromBool(
      Bool=$User.bIsFrontendPreview
      True Float=&RandomRangeFloat(Minimum=0.15, Maximum=0.5)
      False Float=&RandomRangeFloat(Minimum=0.15, Maximum=1)
     )
     4=&RandomRangeFloat(Minimum=0.15, Maximum=0.6)
     Index=&SpawnGroup()
     Num=::NewEnumerator3
    )
    Mesh Scale Mode=::NewEnumerator3
    Mesh Uniform Scale=&MakeCustomFloatFromBool(
     Bool=$User.bIsFrontendPreview
     True Float=&RandomRangeFloat(Minimum=2, Maximum=3.5)
     False Float=&RandomRangeFloat(Minimum=3, Maximum=5)
    )
    Sprite Rotation Mode=::NewEnumerator2
    Sprite Size=&MergeHelper_Vector2Switch(
     0=<0 0>
     1=&MakeCustomVector2DFromBool(
      Bool=$User.bIsFrontendPreview
      True Vector2D=&RandomRangeVector2D(Minimum=<5 60>, Maximum=<8 120>)
      False Vector2D=&RandomRangeVector2D(Minimum=<15 100>, Maximum=<5 50>)
     )
     2=&MakeCustomVector2DFromBool(
      Bool=$User.bIsFrontendPreview
      True Vector2D=&Multiply_Vector2D_ByFloat(Float=&RandomRangeFloat(Minimum=3, Maximum=12), Vector2D=<15 15>)
      False Vector2D=&Multiply_Vector2D_ByFloat(Float=&RandomRangeFloat(Minimum=3, Maximum=12), Vector2D=<20 20>)
     )
     3=&MakeCustomVector2DFromBool(
      Bool=$User.bIsFrontendPreview
      True Vector2D=&Multiply_Vector2D_ByFloat(Float=&RandomRangeFloat(Minimum=5, Maximum=20), Vector2D=<1.5 1.5>)
      False Vector2D=&Multiply_Vector2D_ByFloat(Float=&RandomRangeFloat(Minimum=1, Maximum=10), Vector2D=<12 12>)
     )
     4=&Multiply_Vector2D_ByFloat(Float=&RandomRangeFloat(Minimum=0.5, Maximum=6), Vector2D=<2 2>)
     Index=&SpawnGroup()
     Num=::NewEnumerator3
    )
    Sprite Size Mode=::NewEnumerator4
    UsePositionOffset=true
    Write Material Random=true
   )
   &PartitionParticles(Number Of Partitions=$Emitter.PortCount)
   &SkeletalMeshLocation(
    Bone / Socket Index=$Particles.PartitionParticles.Partition
    Bone / Socket Sampling Mode=::NewEnumerator5
    MaskSpawnGroup=false
    Mesh Sampling Type=::NewEnumerator6
    Sampled Position Offset=&MergeHelper_VectorSwitch(
     0=/Niagara/DynamicInputs/TypeConversions/MakeCustomPositionFromBool(Bool:NiagaraBool=$User.bIsFrontendPreview, True Position:NiagaraPosition=<-180 0 0>, False Position:NiagaraPosition=<-150 0 0>)
     1=<0 0 0>
     2=<0 0 0>
     3=<0 0 0>
     4=<0 0 0>
     Index=&SpawnGroup()
     Num=::NewEnumerator3
    )
    Skeletal Mesh=$System.SkeletalMesh
   )
   &AddVelocity(
    Cone Angle=&MergeHelper_FloatSwitch(
     0=&MakeCustomFloatFromBool(Bool=$User.bIsFrontendPreview, True Float=10, False Float=50)
     1=12
     2=0
     3=10
     4=10
     Index=&SpawnGroup()
     Num=::NewEnumerator3
    )
    Cone Axis=&MergeHelper_VectorSwitch(
     0=&MakeCustomVectorFromBool(
      Bool=$User.bIsFrontendPreview
      True Vector=&RandomRangeVector(Minimum=<1 0.3 -0.35>, Maximum=<1 -0.3 -0.55>)
      False Vector=&RandomRangeVector(Minimum=<1 0.3 -0.5>, Maximum=<1 -0.3 -0.6>)
     )
     1=<1 0 0>
     2=<1 0 0>
     3=<1 0 0>
     4=<1 0 0>
     Index=&SpawnGroup()
     Num=::NewEnumerator3
    )
    Module.UseConeFalloff=false
    Module.UseVelDistribution=false
    Override Seed=false
    Velocity Mode=::NewEnumerator2
    Velocity Speed=&MakeFloatFromVector(
     Vector=&VC_Mask_Vectors(Coro Value:Vector3f=<-400 0 0>, FrontEnd Value:Vector3f=<-2000 0 0>, FN Value:Vector3f=<-1200 0 0>)
     Channel=::NewEnumerator0
    )
   )
   &ShapeLocation(
    Box Size=&MergeHelper_VectorSwitch(
     0=<500 100 60>
     1=<1 100 75>
     2=<1 100 60>
     3=<0 100 60>
     4=<0 100 60>
     Index=&SpawnGroup()
     Num=::NewEnumerator3
    )
    MaskSpawnGroup=false
    Override Seed=false
    Shape Primitive=::NewEnumerator2
    Surface Only=false
    UseOwnerScale=true
   )
   &InitialMeshOrientation(Rotation=&RandomRangeVector(Minimum=<-1 -1 -1>, Maximum=<1 1 1>), Use Rotation Vector=true)
   /VehicleCosmetics/FX/VC_FXParents/Modules/DLR_ProductPreviews(GameModeID:NiagaraInt32=$System.GameModeID, bPreviewVehicleSpeed:NiagaraBool=$System.bPreviewVehicleSpeed)
   /SetVariables(Particles.VisibilityTag=$Emitter.SpawnGroup) {isSetParameters=true, module=SetVariables_#5}
  ]
  particleUpdateScript [
   &ParticleState()
   &SpriteRotationRate(Rotation Rate=&MergeHelper_FloatSwitch(
     0=0
     1=0
     2=&RandomRangeFloat(Minimum=50, Maximum=120)
     3=&RandomRangeFloat(Minimum=50, Maximum=120)
     4=0
     Index=&SpawnGroup()
     Num=::NewEnumerator3
    ))
   &UpdateMeshOrientation(
    Rotation Rate=&FloatFromCurve(
     FloatCurve=DataInterface() {propertyValues {Curve {keys [
         {arriveTangent=3.0228, interpMode=RCIM_Cubic, leaveTangent=3.0228, tangentMode=RCTM_User, time=0, value=0.1}
         {arriveTangent=0.4, interpMode=RCIM_Cubic, leaveTangent=0.4, time=0.2, value=1}
         <at 1 0.5 RCIM_Cubic>
        ]}}}
     Scale Curve=&RandomRangeFloat(Minimum=20, Maximum=50)
    )
    Rotation Vector=&RandomRangeVector(Minimum=<-1 -1 -1>, Maximum=<1 1 1>)
   )
   &ScaleSpriteSize(
    Scale Sprite Size Mode=::NewEnumerator0
    Uniform Scale Factor=&MergeHelper_FloatSwitch(
     0=1
     1=&FloatFromCurve(FloatCurve=DataInterface() {propertyValues {Curve {
         defaultValue=1
         keys [
          {arriveTangent=1.1071, interpMode=RCIM_Cubic, leaveTangent=1.1071, tangentMode=RCTM_User, time=0, value=0.5}
          {arriveTangent=-2.8393, interpMode=RCIM_Cubic, leaveTangent=-2.8393, tangentMode=RCTM_User, time=0.1, value=1.2}
          {arriveTangent=-1.3333, interpMode=RCIM_Cubic, leaveTangent=-1.3333, time=0.2, value=1}
          <at 1 0 RCIM_Cubic>
         ]
        }}})
     2=&MakeCustomFloatFromBool(
      Bool=$User.bIsFrontendPreview
      True Float=&FloatFromCurve(FloatCurve=DataInterface() {propertyValues {Curve {
          defaultValue=1
          keys [
           {interpMode=RCIM_Cubic, tangentMode=RCTM_User, time=0, value=0}
           {arriveTangent=0.319, interpMode=RCIM_Cubic, leaveTangent=0.319, tangentMode=RCTM_User, time=0.25, value=1}
           <at 1 0 RCIM_Cubic>
          ]
         }}})
      False Float=&FloatFromCurve(FloatCurve=DataInterface() {propertyValues {Curve {
          defaultValue=1
          keys [
           {interpMode=RCIM_Cubic, tangentMode=RCTM_User, time=0, value=0}
           {arriveTangent=0.319, interpMode=RCIM_Cubic, leaveTangent=0.319, tangentMode=RCTM_User, time=0.2, value=0.5}
           {interpMode=RCIM_Cubic, tangentMode=RCTM_User, time=0.5762, value=1}
           <at 1 2 RCIM_Cubic>
          ]
         }}})
     )
     3=&FloatFromCurve(FloatCurve=DataInterface() {propertyValues {Curve {
         defaultValue=1
         keys [
          {interpMode=RCIM_Cubic, tangentMode=RCTM_User, time=0, value=0}
          {arriveTangent=0.319, interpMode=RCIM_Cubic, leaveTangent=0.319, tangentMode=RCTM_User, time=0.216, value=0.9805}
          {interpMode=RCIM_Cubic, tangentMode=RCTM_User, time=0.5762, value=0.9223}
         ]
        }}})
     4=1
     Index=&SpawnGroup()
     Num=::NewEnumerator3
    )
   )
   &ScaleMeshSize(Scale Factor=&VectorFromFloat(Value=&FloatFromCurve(FloatCurve=DataInterface() {propertyValues {Curve {keys [<at 0 0>,<at 0.15 1>,<at 0.85 1>,<at 1 0>]}}})))
   &AccelerationForce(Acceleration=&MergeHelper_VectorSwitch(
     0=&MakeCustomVectorFromBool(
      Bool=$User.bIsFrontendPreview
      True Vector=<0 0 0>
      False Vector=&RandomRangeVector(Minimum=<0 0 500>, Maximum=<0 0 800>)
     )
     1=<0 0 0>
     2=<0 0 0>
     3=<0 0 0>
     Index=&SpawnGroup()
     Num=::NewEnumerator2
    ))
   &Collision(
    Age Colliding Particles=false
    Collision Enabled=/Niagara/DynamicInputs/Arrays/SelectBoolFromArray(
     Array Sampling Mode:ENiagara_LifetimeMode=::NewEnumerator0
     Direct Array Index:NiagaraInt32=&SpawnGroup()
     Direct Array Mode:ENiagara_WrapClamp=::NewEnumerator0
     Evaluation Type:ENiagaraRandomnessEvaluation=::NewEnumerator0
     Recalculate Random Each Loop:NiagaraBool=false
     Randomness Mode:ENiagaraRandomnessMode=::NewEnumerator0
     Random Seed:NiagaraInt32=?
     Fixed Random Seed:NiagaraBool=false
     Bool Selection Array:NiagaraDataInterfaceArrayBool=DataInterface() {propertyValues {GpuSyncMode=SyncCpuToGpu, MaxElements=0, BoolData [true,false,false]}}
    )
    Enable Rest State=false
    Friction=1.5
    KillOnCollision=false
    Max CPU Trace Length=2000
    Module.EnableMaxCollisionCount=false
    Module.Randomize Collision Normal Vector=false
    Radius Calculation Type=::NewEnumerator1
    Restitution=&RandomRangeFloat(Minimum=0.5, Maximum=1)
   )
   &Drag(
    Drag=&MergeHelper_FloatSwitch(
     0=&RandomRangeFloat(Minimum=2, Maximum=1.5)
     1=0
     2=0
     3=0
     Index=&SpawnGroup()
     Num=::NewEnumerator2
    )
    Use Linear Drag=true
    Use Rotational Drag=true
   )
   &ScaleColor(
    Scale Alpha=&MergeHelper_FloatSwitch(
     0=1
     1=&FloatFromCurve(FloatCurve=DataInterface() {propertyValues {Curve {keys [
          <at 0 0 RCIM_Cubic>
          <at 0.15 1.0395 RCIM_Cubic>
          {arriveTangent=-3.2314, interpMode=RCIM_Cubic, leaveTangent=-3.2314, tangentMode=RCTM_User, time=1, value=0}
         ]}}})
     2=&MakeCustomFloatFromBool(
      Bool=$User.bIsFrontendPreview
      True Float=&FloatFromCurve(FloatCurve=DataInterface() {propertyValues {Curve {keys [<at 0.25 0.6>,<at 1 1>,<at 1 1>]}}}, Scale Curve=0.1)
      False Float=&FloatFromCurve(FloatCurve=DataInterface() {propertyValues {Curve {keys [
           <at 0 0 RCIM_Cubic>
           {arriveTangent=-0.1006, interpMode=RCIM_Cubic, leaveTangent=-0.1006, tangentMode=RCTM_User, time=0.2, value=0.1}
           {arriveTangent=-0.0394, interpMode=RCIM_Cubic, leaveTangent=-0.0394, time=0.587, value=0.1323}
           <at 1.0121 0.068 RCIM_Cubic>
          ]}}})
     )
     3=&MakeCustomFloatFromBool(
      Bool=$User.bIsFrontendPreview
      True Float=&FloatFromCurve(FloatCurve=DataInterface() {propertyValues {Curve {keys [<at 0 0>,<at 0.15 1>,<at 0.5 0>]}}}, Scale Curve=0.5)
      False Float=&FloatFromCurve(FloatCurve=DataInterface() {propertyValues {Curve {keys [<at 0 0>,<at 0.15 1>,<at 1 1>]}}}, Scale Curve=0.5)
     )
     4=&FloatFromCurve(FloatCurve=DataInterface() {propertyValues {Curve {keys [<at 0.15 0>,<at 0.25 1>,<at 0.35 0>]}}})
     Index=&SpawnGroup()
     Num=::NewEnumerator3
    )
    Scale RGB=&MergeHelper_VectorSwitch(
     0=<1 1 1>
     1=&VectorFromFloat(Value=&FloatFromCurve(FloatCurve=DataInterface() {propertyValues {Curve {keys [<at 0 5>,<at 1.0448 -1.5444>]}}}))
     2=&MakeCustomVectorFromBool(
      Bool=$User.bIsFrontendPreview
      True Vector=<1 1 1>
      False Vector=&VectorFromFloat(Value=&FloatFromCurve(FloatCurve=DataInterface() {propertyValues {Curve {keys [<at 0.5 0.5>,<at 1 1.5>]}}}))
     )
     3=<1 1 1>
     Index=&SpawnGroup()
     Num=::NewEnumerator2
    )
    ScaleA=true
    ScaleRGB=true
   )
   &InheritSourceMovement(
    Applied Position Scale=&VectorFromFloat(Value=&FloatFromCurve(FloatCurve=DataInterface() {propertyValues {Curve {
         defaultValue=1
         keys [
          {arriveTangent=-1.7885, interpMode=RCIM_Cubic, leaveTangent=-1.7885, tangentMode=RCTM_User, time=0.05, value=1}
          {interpMode=RCIM_Cubic, tangentMode=RCTM_User, time=1, value=0}
         ]
        }}}))
    ApplyForce=false
    ApplyPosition=true
    ApplyPositionFalloff=false
   )
   &GravityForce(Gravity=&MergeHelper_VectorSwitch(
     0=&MakeCustomVectorFromBool(
      Bool=$User.bIsFrontendPreview
      True Vector=&RandomRangeVector(Minimum=<0 0 -1000>, Maximum=<0 0 -3000>)
      False Vector=&RandomRangeVector(Minimum=<0 0 -2000>, Maximum=<0 0 -3000>)
     )
     1=<0 0 0>
     2=&MakeCustomVectorFromBool(Bool=$User.bIsFrontendPreview, True Vector=<0 0 0>, False Vector=<0 0 500>)
     3=<0 0 0>
     4=<0 0 1000>
     Index=&SpawnGroup()
     Num=::NewEnumerator3
    ))
   &SolveForcesAndVelocity(Clamp Velocity=false, Limit Acceleration=false)
  ]
  rendererClasses [&NiagaraMeshRendererProperties, &NiagaraSpriteRendererProperties]
  renderers [
   {rendererClass=&NiagaraMeshRendererProperties, rendererIndex=0}
   {rendererClass=&NiagaraSpriteRendererProperties, rendererIndex=1}
   {rendererClass=&NiagaraSpriteRendererProperties, rendererIndex=2}
   {rendererClass=&NiagaraSpriteRendererProperties, rendererIndex=3}
   {rendererClass=&NiagaraSpriteRendererProperties, rendererIndex=4}
  ]
 }
 NE_VC_Boost_Sprite01001 {
  emitterUpdateScript [
   &EmitterState()
   &VC_FilteredSocketCount()
   &SpawnRate(
    SpawnRate=&Multiply_Float(
     A=&MakeCustomFloatFromBool(Bool=$User.bIsFrontendPreview, True Float=65, False Float=40)
     B=&VC_Mask_Floats(Coro Value:NiagaraFloat=1, FrontEnd Value:NiagaraFloat=0.3, FN Value:NiagaraFloat=0.5)
    )
    Use Spawn Probability=false
   )
   &SpawnRate(
    SpawnGroup=1
    SpawnRate=&Multiply_Float(
     A=&MakeCustomFloatFromBool(Bool=$User.bIsFrontendPreview, True Float=100, False Float=200)
     B=&VC_Mask_Floats(Coro Value:NiagaraFloat=1, FrontEnd Value:NiagaraFloat=0.5, FN Value:NiagaraFloat=0.5)
    )
    Use Spawn Probability=false
   ) {module=SpawnRate001}
   &SpawnRate(
    SpawnGroup=2
    SpawnRate=&Multiply_Float(
     A=400
     B=&VC_Mask_Floats(Coro Value:NiagaraFloat=1, FrontEnd Value:NiagaraFloat=0.3, FN Value:NiagaraFloat=0.5)
    )
    Use Spawn Probability=false
   ) {module=SpawnRate002}
   &SpawnRate(
    SpawnGroup=3
    SpawnRate=&Multiply_Float(
     A=150
     B=&VC_Mask_Floats(Coro Value:NiagaraFloat=1, FrontEnd Value:NiagaraFloat=0.5, FN Value:NiagaraFloat=0.5)
    )
    Use Spawn Probability=false
   ) {module=SpawnRate003}
   /SetVariables(
    Emitter.DustColor:LinearColor=&MakeCustomLinearColorFromBool(Bool=$User.bIsFrontendPreview, True Color=<rgba 11 5.5 0 1>, False Color=<rgba 50 18.2 10 1>)
    Emitter.Sand Alpha Power:NiagaraFloat=&MakeCustomFloatFromBool(Bool=$User.bIsFrontendPreview, True Float=3, False Float=5)
   ) {isSetParameters=true, module=SetVariables_#4}
  ]
  particleSpawnScript [
   &VC_PaintedVariantsOverride() {enabled=false}
   &InitializeParticle(
    Color=&MergeHelper_ColorSwitch(
     0=<rgba 1 1 1 0>
     1=&MakeCustomLinearColorFromBool(Bool=$User.bIsFrontendPreview, True Color=<rgba 100 83.3333 0 1>, False Color=<rgba 5 1.6667 0 10>)
     2=&MakeCustomLinearColorFromBool(Bool=$User.bIsFrontendPreview, True Color=<rgba 1.5 0.7039 0 1>, False Color=<rgba 30 13.5 12 1>)
     3=&MakeCustomLinearColorFromBool(Bool=$User.bIsFrontendPreview, True Color=<rgba 10 4.5 4 1>, False Color=<rgba 10 3.125 2.5 0.1>)
     4=&MakeCustomLinearColorFromBool(Bool=$User.bIsFrontendPreview, True Color=<rgba 30 15.9977 15 1>, False Color=<rgba 500 88.5984 31.145 1>)
     Index=&SpawnGroup()
     Num=::NewEnumerator2
    )
    Lifetime=&MergeHelper_FloatSwitch(
     0=&MakeCustomFloatFromBool(
      Bool=$User.bIsFrontendPreview
      True Float=&RandomRangeFloat(Minimum=0.3, Maximum=1)
      False Float=&RandomRangeFloat(Minimum=0.5, Maximum=1.5)
     )
     1=&MakeCustomFloatFromBool(
      Bool=$User.bIsFrontendPreview
      True Float=&RandomRangeFloat(Minimum=0.1, Maximum=0.3)
      False Float=&RandomRangeFloat(Minimum=0.1, Maximum=0.3)
     )
     2=&RandomRangeFloat(Minimum=0.15, Maximum=0.7)
     3=&MakeCustomFloatFromBool(
      Bool=$User.bIsFrontendPreview
      True Float=&RandomRangeFloat(Minimum=0.15, Maximum=0.5)
      False Float=&RandomRangeFloat(Minimum=0.15, Maximum=0.5)
     )
     4=&RandomRangeFloat(Minimum=0.15, Maximum=0.6)
     Index=&SpawnGroup()
     Num=::NewEnumerator2
    )
    Mesh Scale Mode=::NewEnumerator3
    Mesh Uniform Scale=&MakeCustomFloatFromBool(
     Bool=$User.bIsFrontendPreview
     True Float=&RandomRangeFloat(Minimum=2, Maximum=3.5)
     False Float=&RandomRangeFloat(Minimum=3, Maximum=5)
    )
    Sprite Rotation Mode=::NewEnumerator2
    Sprite Size=&MergeHelper_Vector2Switch(
     0=<0 0>
     1=&MakeCustomVector2DFromBool(
      Bool=$User.bIsFrontendPreview
      True Vector2D=&RandomRangeVector2D(Minimum=<5 60>, Maximum=<8 120>)
      False Vector2D=&RandomRangeVector2D(Minimum=<15 100>, Maximum=<5 50>)
     )
     2=&MakeCustomVector2DFromBool(
      Bool=$User.bIsFrontendPreview
      True Vector2D=&Multiply_Vector2D_ByFloat(Float=&RandomRangeFloat(Minimum=3, Maximum=12), Vector2D=<15 15>)
      False Vector2D=&Multiply_Vector2D_ByFloat(Float=&RandomRangeFloat(Minimum=3, Maximum=12), Vector2D=<20 20>)
     )
     3=&MakeCustomVector2DFromBool(
      Bool=$User.bIsFrontendPreview
      True Vector2D=&Multiply_Vector2D_ByFloat(Float=&RandomRangeFloat(Minimum=5, Maximum=20), Vector2D=<1.5 1.5>)
      False Vector2D=&Multiply_Vector2D_ByFloat(Float=&RandomRangeFloat(Minimum=1, Maximum=10), Vector2D=<12 12>)
     )
     4=&Multiply_Vector2D_ByFloat(Float=&RandomRangeFloat(Minimum=0.5, Maximum=6), Vector2D=<2 2>)
     Index=&SpawnGroup()
     Num=::NewEnumerator2
    )
    Sprite Size Mode=::NewEnumerator4
    UsePositionOffset=true
    Write Material Random=true
   )
   &PartitionParticles(Number Of Partitions=$Emitter.PortCount)
   &SkeletalMeshLocation(
    Bone / Socket Index=$Particles.PartitionParticles.Partition
    Bone / Socket Sampling Mode=::NewEnumerator5
    MaskSpawnGroup=false
    Mesh Sampling Type=::NewEnumerator6
    Sampled Position Offset=&MergeHelper_VectorSwitch(
     0=&MakeCustomPositionFromBool(Bool:NiagaraBool=$User.bIsFrontendPreview, True Position:NiagaraPosition=<-180 0 0>, False Position:NiagaraPosition=<-150 0 0>)
     1=<0 0 0>
     2=<0 0 0>
     3=<0 0 0>
     Index=&SpawnGroup()
     Num=::NewEnumerator2
    )
    Skeletal Mesh=$System.SkeletalMesh
   )
   &AddVelocity(
    Cone Angle=&MergeHelper_FloatSwitch(
     0=&MakeCustomFloatFromBool(Bool=$User.bIsFrontendPreview, True Float=10, False Float=50)
     1=12
     2=0
     3=10
     Index=&SpawnGroup()
     Num=::NewEnumerator2
    )
    Cone Axis=&MergeHelper_VectorSwitch(
     0=&MakeCustomVectorFromBool(
      Bool=$User.bIsFrontendPreview
      True Vector=&RandomRangeVector(Minimum=<1 0.3 -0.3>, Maximum=<1 -0.3 -0.55>)
      False Vector=&RandomRangeVector(Minimum=<1 0.3 -0.5>, Maximum=<1 -0.3 -0.6>)
     )
     1=<1 0 0>
     2=<1 0 0>
     3=<1 0 0>
     Index=&SpawnGroup()
     Num=::NewEnumerator2
    )
    Module.UseConeFalloff=false
    Module.UseVelDistribution=false
    Override Seed=false
    Velocity Mode=::NewEnumerator2
    Velocity Speed=&MakeFloatFromVector(
     Vector=&VC_Mask_Vectors(Coro Value:Vector3f=<-400 0 0>, FrontEnd Value:Vector3f=<-2000 0 0>, FN Value:Vector3f=<-1200 0 0>)
     Channel=::NewEnumerator0
    )
   )
   &ShapeLocation(
    Box Size=&MergeHelper_VectorSwitch(0=<500 100 60>, 1=<1 100 75>, 2=<1 100 60>, 3=<0 100 60>, Index=&SpawnGroup(), Num=::NewEnumerator2)
    MaskSpawnGroup=false
    Override Seed=false
    Shape Primitive=::NewEnumerator2
    Surface Only=false
    UseOwnerScale=true
   )
   &InitialMeshOrientation(Rotation=&RandomRangeVector(Minimum=<-1 -1 -1>, Maximum=<1 1 1>), Use Rotation Vector=true)
   &DLR_ProductPreviews(GameModeID:NiagaraInt32=$System.GameModeID, bPreviewVehicleSpeed:NiagaraBool=$System.bPreviewVehicleSpeed)
   /SetVariables(Particles.VisibilityTag=$Emitter.SpawnGroup) {isSetParameters=true, module=SetVariables_#5}
  ]
  particleUpdateScript [
   &ParticleState()
   &SpriteRotationRate(Rotation Rate=&MergeHelper_FloatSwitch(
     0=0
     1=0
     2=&RandomRangeFloat(Minimum=50, Maximum=120)
     3=&RandomRangeFloat(Minimum=50, Maximum=120)
     Index=&SpawnGroup()
     Num=::NewEnumerator2
    ))
   &UpdateMeshOrientation(
    Rotation Rate=&FloatFromCurve(
     FloatCurve=DataInterface() {propertyValues {Curve {keys [
         {arriveTangent=3.0228, interpMode=RCIM_Cubic, leaveTangent=3.0228, tangentMode=RCTM_User, time=0, value=0.1}
         {arriveTangent=0.4, interpMode=RCIM_Cubic, leaveTangent=0.4, time=0.2, value=1}
         <at 1 0.5 RCIM_Cubic>
        ]}}}
     Scale Curve=&RandomRangeFloat(Minimum=20, Maximum=50)
    )
    Rotation Vector=&RandomRangeVector(Minimum=<-1 -1 -1>, Maximum=<1 1 1>)
   )
   &ScaleSpriteSize(
    Scale Sprite Size Mode=::NewEnumerator0
    Uniform Scale Factor=&MergeHelper_FloatSwitch(
     0=1
     1=&FloatFromCurve(FloatCurve=DataInterface() {propertyValues {Curve {
         defaultValue=1
         keys [
          {arriveTangent=1.1071, interpMode=RCIM_Cubic, leaveTangent=1.1071, tangentMode=RCTM_User, time=0, value=0.5}
          {arriveTangent=-2.8393, interpMode=RCIM_Cubic, leaveTangent=-2.8393, tangentMode=RCTM_User, time=0.1, value=1.2}
          {arriveTangent=-1.3333, interpMode=RCIM_Cubic, leaveTangent=-1.3333, time=0.2, value=1}
          <at 1 0 RCIM_Cubic>
         ]
        }}})
     2=&MakeCustomFloatFromBool(
      Bool=$User.bIsFrontendPreview
      True Float=&FloatFromCurve(FloatCurve=DataInterface() {propertyValues {Curve {
          defaultValue=1
          keys [
           {interpMode=RCIM_Cubic, tangentMode=RCTM_User, time=0, value=0}
           {arriveTangent=0.319, interpMode=RCIM_Cubic, leaveTangent=0.319, tangentMode=RCTM_User, time=0.25, value=1}
           <at 1 0 RCIM_Cubic>
          ]
         }}})
      False Float=&FloatFromCurve(FloatCurve=DataInterface() {propertyValues {Curve {
          defaultValue=1
          keys [
           {interpMode=RCIM_Cubic, tangentMode=RCTM_User, time=0, value=0}
           {arriveTangent=0.319, interpMode=RCIM_Cubic, leaveTangent=0.319, tangentMode=RCTM_User, time=0.2, value=0.5}
           {interpMode=RCIM_Cubic, tangentMode=RCTM_User, time=0.5762, value=1}
           <at 1 2 RCIM_Cubic>
          ]
         }}})
     )
     3=&FloatFromCurve(FloatCurve=DataInterface() {propertyValues {Curve {
         defaultValue=1
         keys [
          {interpMode=RCIM_Cubic, tangentMode=RCTM_User, time=0, value=0}
          {arriveTangent=0.319, interpMode=RCIM_Cubic, leaveTangent=0.319, tangentMode=RCTM_User, time=0.216, value=0.9805}
          {interpMode=RCIM_Cubic, tangentMode=RCTM_User, time=0.5762, value=0.9223}
         ]
        }}})
     Index=&SpawnGroup()
     Num=::NewEnumerator2
    )
   )
   &ScaleMeshSize(Scale Factor=&VectorFromFloat(Value=&FloatFromCurve(FloatCurve=DataInterface() {propertyValues {Curve {keys [<at 0 0>,<at 0.15 1>,<at 0.85 1>,<at 1 0>]}}})))
   &AccelerationForce(Acceleration=&MergeHelper_VectorSwitch(
     0=&MakeCustomVectorFromBool(
      Bool=$User.bIsFrontendPreview
      True Vector=<0 0 0>
      False Vector=&RandomRangeVector(Minimum=<0 0 500>, Maximum=<0 0 800>)
     )
     1=<0 0 0>
     2=<0 0 0>
     3=<0 0 0>
     Index=&SpawnGroup()
     Num=::NewEnumerator2
    ))
   &Collision(
    Age Colliding Particles=false
    Collision Enabled=&SelectBoolFromArray(
     Array Sampling Mode:ENiagara_LifetimeMode=::NewEnumerator0
     Direct Array Index:NiagaraInt32=&SpawnGroup()
     Direct Array Mode:ENiagara_WrapClamp=::NewEnumerator0
     Evaluation Type:ENiagaraRandomnessEvaluation=::NewEnumerator0
     Recalculate Random Each Loop:NiagaraBool=false
     Randomness Mode:ENiagaraRandomnessMode=::NewEnumerator0
     Random Seed:NiagaraInt32=?
     Fixed Random Seed:NiagaraBool=false
     Bool Selection Array:NiagaraDataInterfaceArrayBool=DataInterface() {propertyValues {GpuSyncMode=SyncCpuToGpu, MaxElements=0, BoolData [true,false,false]}}
    )
    Enable Rest State=false
    Friction=1.5
    KillOnCollision=false
    Max CPU Trace Length=2000
    Module.EnableMaxCollisionCount=false
    Module.Randomize Collision Normal Vector=false
    Radius Calculation Type=::NewEnumerator1
    Restitution=&RandomRangeFloat(Minimum=0.5, Maximum=1)
   )
   &Drag(
    Drag=&MergeHelper_FloatSwitch(
     0=&RandomRangeFloat(Minimum=2, Maximum=1.5)
     1=0
     2=0
     3=0
     Index=&SpawnGroup()
     Num=::NewEnumerator2
    )
    Use Linear Drag=true
    Use Rotational Drag=true
   )
   &ScaleColor(
    Scale Alpha=&MergeHelper_FloatSwitch(
     0=1
     1=&FloatFromCurve(FloatCurve=DataInterface() {propertyValues {Curve {keys [
          <at 0 0 RCIM_Cubic>
          <at 0.15 1.0395 RCIM_Cubic>
          {arriveTangent=-3.2314, interpMode=RCIM_Cubic, leaveTangent=-3.2314, tangentMode=RCTM_User, time=1, value=0}
         ]}}})
     2=&MakeCustomFloatFromBool(
      Bool=$User.bIsFrontendPreview
      True Float=&FloatFromCurve(FloatCurve=DataInterface() {propertyValues {Curve {keys [<at 0.25 0.6>,<at 1 1>,<at 1 1>]}}}, Scale Curve=0.1)
      False Float=&FloatFromCurve(FloatCurve=DataInterface() {propertyValues {Curve {keys [
           <at 0 0 RCIM_Cubic>
           {arriveTangent=-0.1006, interpMode=RCIM_Cubic, leaveTangent=-0.1006, tangentMode=RCTM_User, time=0.2, value=0.1}
           {arriveTangent=-0.1667, interpMode=RCIM_Cubic, leaveTangent=-0.1667, time=0.587, value=0.1}
           <at 0.8 0 RCIM_Cubic>
          ]}}})
     )
     3=&MakeCustomFloatFromBool(
      Bool=$User.bIsFrontendPreview
      True Float=&FloatFromCurve(FloatCurve=DataInterface() {propertyValues {Curve {keys [<at 0 0>,<at 0.15 1>,<at 0.5 0>]}}}, Scale Curve=0.5)
      False Float=&FloatFromCurve(FloatCurve=DataInterface() {propertyValues {Curve {keys [<at 0 0>,<at 0.15 1>,<at 1 0>]}}}, Scale Curve=0.5)
     )
     4=&FloatFromCurve(FloatCurve=DataInterface() {propertyValues {Curve {keys [<at 0.15 0>,<at 0.25 1>,<at 0.35 0>]}}})
     Index=&SpawnGroup()
     Num=::NewEnumerator2
    )
    Scale RGB=&MergeHelper_VectorSwitch(
     0=<1 1 1>
     1=&VectorFromFloat(Value=&FloatFromCurve(FloatCurve=DataInterface() {propertyValues {Curve {keys [<at 0 5>,<at 1.0448 -1.5444>]}}}))
     2=<1 1 1>
     3=<1 1 1>
     Index=&SpawnGroup()
     Num=::NewEnumerator2
    )
    ScaleA=true
    ScaleRGB=true
   )
   &InheritSourceMovement(
    Applied Position Scale=&VectorFromFloat(Value=&FloatFromCurve(FloatCurve=DataInterface() {propertyValues {Curve {
         defaultValue=1
         keys [
          {arriveTangent=-1.7885, interpMode=RCIM_Cubic, leaveTangent=-1.7885, tangentMode=RCTM_User, time=0.05, value=1}
          {interpMode=RCIM_Cubic, tangentMode=RCTM_User, time=1, value=0}
         ]
        }}}))
    ApplyForce=false
    ApplyPosition=true
    ApplyPositionFalloff=false
   )
   &GravityForce(Gravity=&MergeHelper_VectorSwitch(
     0=&MakeCustomVectorFromBool(
      Bool=$User.bIsFrontendPreview
      True Vector=&RandomRangeVector(Minimum=<0 0 -1000>, Maximum=<0 0 -3000>)
      False Vector=&RandomRangeVector(Minimum=<0 0 -2000>, Maximum=<0 0 -3000>)
     )
     1=<0 0 0>
     2=&MakeCustomVectorFromBool(Bool=$User.bIsFrontendPreview, True Vector=<0 0 0>, False Vector=<0 0 500>)
     3=<0 0 0>
     Index=&SpawnGroup()
     Num=::NewEnumerator2
    ))
   &SolveForcesAndVelocity(Clamp Velocity=false, Limit Acceleration=false)
  ]
  rendererClasses [&NiagaraMeshRendererProperties, &NiagaraSpriteRendererProperties]
  renderers [
   {rendererClass=&NiagaraMeshRendererProperties, rendererIndex=0}
   {rendererClass=&NiagaraSpriteRendererProperties, rendererIndex=1}
   {rendererClass=&NiagaraSpriteRendererProperties, rendererIndex=2}
   {rendererClass=&NiagaraSpriteRendererProperties, rendererIndex=3}
  ]
 }
 NE_VC_Boost_Sprite02 {
  emitterUpdateScript [
   &EmitterState()
   &VC_FilteredSocketCount()
   &SpawnRate(
    SpawnRate=&Multiply_Float(
     A=600
     B=&VC_Mask_Floats(Coro Value:NiagaraFloat=1, FrontEnd Value:NiagaraFloat=0.4, FN Value:NiagaraFloat=0.5)
    )
    Use Spawn Probability=false
   )
  ]
  particleSpawnScript [
   &VC_PaintedVariantsOverride() {enabled=false}
   &InitializeParticle(
    Color=&MakeCustomLinearColorFromBool(Bool=$User.bIsFrontendPreview, True Color=<rgba 500 170 5 1>, False Color=<rgba 500 104 5 1>)
    Lifetime=&MakeCustomFloatFromBool(
     Bool=$User.bIsFrontendPreview
     True Float=&RandomRangeFloat(Minimum=0.1, Maximum=0.5)
     False Float=&RandomRangeFloat(Minimum=0.1, Maximum=0.3)
    )
    Sprite Rotation Mode=::NewEnumerator2
    Sprite Size=&MakeCustomVector2DFromBool(
     Bool=$User.bIsFrontendPreview
     True Vector2D=&RandomRangeVector2D(Minimum=<2 50>, Maximum=<10 20>)
     False Vector2D=&RandomRangeVector2D(Minimum=<10 30>, Maximum=<5 70>)
    )
    Sprite Size Mode=::NewEnumerator4
    UsePositionOffset=false
    Write Material Random=true
   )
   &PartitionParticles(Number Of Partitions=$Emitter.PortCount)
   &SkeletalMeshLocation(
    Bone / Socket Index=$Particles.PartitionParticles.Partition
    Bone / Socket Sampling Mode=::NewEnumerator5
    MaskSpawnGroup=false
    Mesh Sampling Type=::NewEnumerator6
    Skeletal Mesh=$System.SkeletalMesh
   )
   &AddVelocity(
    Cone Angle=&MakeCustomFloatFromBool(Bool=$User.bIsFrontendPreview, True Float=88, False Float=160)
    Cone Axis=&RandomRangeVector(Minimum=<1 0.1 -0.5>, Maximum=<1 -0.1 0.5>)
    Inner Cone Angle=30
    Module.UseConeFalloff=false
    Module.UseVelDistribution=false
    Override Seed=false
    Velocity Mode=::NewEnumerator2
    Velocity Speed=&MakeFloatFromVector(
     Vector=&VC_Mask_Vectors(
      Coro Value:Vector3f=<-100 0 0>
      FrontEnd Value:Vector3f=&MakeCustomPositionFromBool(Bool:NiagaraBool=$User.bIsFrontendPreview, True Position:NiagaraPosition=<-550 0 0>, False Position:NiagaraPosition=<-1200 0 0>)
      FN Value:Vector3f=<-1200 0 0>
     )
     Channel=::NewEnumerator0
    )
   )
   &DLR_ProductPreviews(GameModeID:NiagaraInt32=$System.GameModeID, bPreviewVehicleSpeed:NiagaraBool=$System.bPreviewVehicleSpeed)
  ]
  particleUpdateScript [
   &ParticleState()
   &ScaleSpriteSize(
    Scale Sprite Size Mode=::NewEnumerator0
    Uniform Scale Factor=&FloatFromCurve(FloatCurve=DataInterface() {propertyValues {Curve {
        defaultValue=1
        keys [
         {arriveTangent=1.1071, interpMode=RCIM_Cubic, leaveTangent=1.1071, tangentMode=RCTM_User, time=0, value=0.1}
         {arriveTangent=-3.1423, interpMode=RCIM_Cubic, leaveTangent=-3.1423, tangentMode=RCTM_User, time=0.3, value=1.5}
         {interpMode=RCIM_Cubic, leaveTangent=0.1667, time=1, value=0}
        ]
       }}})
   )
   &ScaleColor(
    Scale Alpha=&FloatFromCurve(FloatCurve=DataInterface() {propertyValues {Curve {keys [<at 0 1 RCIM_Cubic>, {arriveTangent=-3.2314, interpMode=RCIM_Cubic, leaveTangent=-3.2314, tangentMode=RCTM_User, time=1, value=0}]}}})
    Scale RGB=&Multiply_VectorByFloat(
     Vector=&VectorFromCurve(VectorCurve=DataInterface() {propertyValues {XCurve {keys [<at 0.45 1>,<at 1 3>]}, YCurve {keys [<at 0 1.3>,<at 1 0.9>]}, ZCurve {keys [<at 0 1.1>,<at 1 1.1>]}}})
     Float=1
    )
    ScaleA=true
    ScaleRGB=true
   )
   &InheritSourceMovement(
    Applied Position Scale=&VectorFromFloat(Value=&FloatFromCurve(FloatCurve=DataInterface() {propertyValues {Curve {
         defaultValue=1
         keys [
          {arriveTangent=-1.7885, interpMode=RCIM_Cubic, leaveTangent=-1.7885, tangentMode=RCTM_User, time=0.05, value=1}
          {interpMode=RCIM_Cubic, tangentMode=RCTM_User, time=1, value=0}
         ]
        }}}))
    ApplyForce=false
    ApplyPosition=true
    ApplyPositionFalloff=false
   )
   &CurlNoiseForce(
    Bias Noise=false
    Noise Frequency=&RandomRangeFloat(Minimum=120, Maximum=1)
    Noise Strength=&RandomRangeFloat(Minimum=2300, Maximum=1000)
    Pan Noise=false
   )
   &SolveForcesAndVelocity(Clamp Velocity=false, Limit Acceleration=false)
  ]
  rendererClasses [&NiagaraSpriteRendererProperties]
  renderers [{rendererClass=&NiagaraSpriteRendererProperties}]
 }
}
summary {
 systemName=NS_Alomnd
 emitters [
  {emitterName=NE_VC_Boost_Sprite02, rendererClasses [&NiagaraSpriteRendererProperties]}
  {
   bEnabled=false
   emitterName=NE_VC_Boost_Sprite01
   rendererClasses [&NiagaraMeshRendererProperties, &NiagaraSpriteRendererProperties]
  }
  {
   emitterName=NE_VC_Boost_Sprite01001
   rendererClasses [&NiagaraMeshRendererProperties, &NiagaraSpriteRendererProperties]
  }
 ]
 userVariables [
  User.bIsFrontendPreview {type=NiagaraBool}
  User.bIsPainted {type=NiagaraBool}
  User.blsIsCoroColor {type=NiagaraBool}
  User.bPreviewVehicleSpeed {type=NiagaraBool}
  User.ColorVariantID {type=NiagaraInt32}
  User.GameModeID {type=NiagaraInt32}
  User.PaintedColor {type=LinearColor, description="", defaultValue {struct=LinearColor, value=<rgba 1 1 1 1>}}
  User.PaintedScalar {type=NiagaraFloat, defaultValue=1, description=""}
  User.PortCount {type=NiagaraInt32, defaultValue=2}
 ]
}
